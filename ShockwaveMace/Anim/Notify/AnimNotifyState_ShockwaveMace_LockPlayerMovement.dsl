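; Anim notify state that stops player movement while swinging the Shockwave Mace, for the duration of the notify window.
; On begin, it casts the mesh owner to a Fort player pawn and applies GE_ShockwaveMace_PreventMove to that pawn's ability system component.
; On end, it removes all stacks of that effect sourced from the same component, with validity checks guarding both paths.
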(fn GetNotifyName
  (return "AnimStateNotify_LockPlayerMovement"))

(fn Received_NotifyBegin
  (params (MeshComp :type Skeletal%20Mesh%20Component%20Object%20Reference) (Animation :type Anim%20Sequence%20Base%20Object%20Reference) (TotalDuration :type Float%20%28single-precision%29) (EventReference :type Anim%20Notify%20Event%20Reference%20Structure%20%28by%20ref%29))
  ;[0.5365 1 0.5365] Apply GE_ShockwaveMace_PreventMove to stop player movement while swinging the ShockwaveMace
  (Utilities|IsValid MeshComp
    (:Is%20Valid
      (bind _asfort_player_pawn (Utilities|Casting|CastToFortPlayerPawn (Components|GetOwner MeshComp))
        (:then
          (Utilities|IsValid (Ability|GetFortAbilitySystemComponent _asfort_player_pawn)
            (:Is%20Valid
              (GameplayEffects|ApplyGameplayEffectToSelf :self (Ability|GetFortAbilitySystemComponent _asfort_player_pawn) :GameplayEffectClass "/ShockwaveMace/GameplayEffects/GE_ShockwaveMace_PreventMove.GE_ShockwaveMace_PreventMove_C" :Level 0.0)
              (return true))
            (:Is%20Not%20Valid
              _)))
        (:CastFailed
          _))
      (return false))
    (:Is%20Not%20Valid
      _)))

(fn Received_NotifyEnd
  (params (MeshComp :type Skeletal%20Mesh%20Component%20Object%20Reference) (Animation :type Anim%20Sequence%20Base%20Object%20Reference) (EventReference :type Anim%20Notify%20Event%20Reference%20Structure%20%28by%20ref%29))
  ;[1 0.4468 0] Remove all stacks of GE_ShockwaveMace_PreventMove
  (Utilities|IsValid MeshComp
    (:Is%20Valid
      (bind _asfort_player_pawn (Utilities|Casting|CastToFortPlayerPawn (Components|GetOwner MeshComp))
        (:then
          (Utilities|IsValid (Ability|GetFortAbilitySystemComponent _asfort_player_pawn)
            (:Is%20Valid
              (GameplayEffects|RemoveActiveGameplayEffectbySourceEffect (Ability|GetFortAbilitySystemComponent _asfort_player_pawn) "/ShockwaveMace/GameplayEffects/GE_ShockwaveMace_PreventMove.GE_ShockwaveMace_PreventMove_C" (Ability|GetFortAbilitySystemComponent _asfort_player_pawn) -1)
              (return true))
            (:Is%20Not%20Valid
              _)))
        (:CastFailed
          _))
      (return false))
    (:Is%20Not%20Valid
      _)))
