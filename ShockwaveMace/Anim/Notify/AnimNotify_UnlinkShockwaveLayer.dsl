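; An anim notify that unlinks the Shockwave Mace animation layer at a chosen point in an animation.
; It gets the owner of the notifying skeletal mesh, finds its ShockwaveMaceComponent, and calls UnlinkAnimLayerClient on it when both the mesh and component are valid.
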(fn Received_Notify
  (params (MeshComp :type Skeletal%20Mesh%20Component%20Object%20Reference) (Animation :type Anim%20Sequence%20Base%20Object%20Reference) (EventReference :type Anim%20Notify%20Event%20Reference%20Structure%20%28by%20ref%29))
  (bind _returnvalue (Components|GetOwner MeshComp))
  (bind _returnvalue_2 (Actor|GetComponentbyClass _returnvalue "/ShockwaveMace/Components/ShockwaveMaceComponent.ShockwaveMaceComponent_C"))
  (Utilities|IsValid MeshComp
    (:Is%20Valid
      (Utilities|IsValid _returnvalue_2
        (:Is%20Valid
          (Class|ShockwaveMaceComponent|UnlinkAnimLayerClient _returnvalue_2)
          (return true))))))
